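; Juice bottle case prop Blueprint that applies a shared tint color to its bottle container meshes.
; The construction script runs the parent's logic, then loops over six static mesh components and passes each, with TintColor, to SetColoronSecondaryMesh.
; BeginPlay, overlap and Tick events are empty, so this Blueprint adds no runtime logic of its own.
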(event EventBeginPlay)

(event Collision|EventActorBeginOverlap
  (params (OtherActor :type Actor%20Object%20Reference)))

(event EventTick
  (params (DeltaSeconds :type Float%20%28single-precision%29)))

(fn ConstructionScript
  (|CallParentFunction)
  (for _array_element (Utilities|Array|MakeArray (Variables|Default|GetSM_Container_JuiceBottles_C4) (Variables|Default|GetSM_Container_JuiceBottles_C3) (Variables|Default|GetSM_Container_JuiceBottles_C2) (Variables|Default|GetSM_Container_JuiceBottles_C1) (Variables|Default|GetSM_Container_JuiceBottles_C) (Variables|Default|GetSM_Container_JuiceBottles_C5))
    (Default|SetColoronSecondaryMesh (Variables|TintColor|GetTintColor) _array_element)))
